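# Physics asset for the PalmTree pickaxe, a melee weapon cosmetic, that defines secondary motion for two fur pieces and two wires.
# Each piece pairs a tiny capsule body on a base bone with a larger, offset capsule on a dyn bone, joined by one constraint.
# Fur constraints have tight swing limits and no twist, while wire constraints swing widely on one axis and barely on the others, giving hinge-like dangle.
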PhysicsAsset /BRCosmetics/Weapons/FORT_Melee/Pickaxe_PalmTree/Meshes/PalmTree_Axe_Physics
bodies {
 base_fur_00 [{shapeType=Capsule, length=0.505, radius=0.505, shapeName=base_fur_00_capsule, rotation=<rot 0 -90 0>}]
 base_fur_01 [{shapeType=Capsule, length=0.505, radius=0.505, shapeName=base_fur_01_capsule, rotation=<rot 0 -90 0>}]
 base_wire_dwn [{shapeType=Capsule, length=0.505, radius=0.505, shapeName=base_wire_dwn_capsule, rotation=<rot 0 -90 0>}]
 base_wire_up [{shapeType=Capsule, length=0.505, radius=0.505, shapeName=base_wire_up_capsule, rotation=<rot 0 -90 0>}]
 dyn_fur_00 [{shapeType=Capsule, length=2, radius=1.5, shapeName=dyn_fur_00_capsule, center=<1.2 0 1.2>, rotation=<rot 90 0 0>}]
 dyn_fur_01 [{shapeType=Capsule, length=2, radius=2, shapeName=dyn_fur_01_capsule, center=<-1.3 0 1.3>, rotation=<rot 90 0 0>}]
 dyn_wire_dwn [{shapeType=Capsule, length=2, radius=1, shapeName=dyn_wire_dwn_capsule, center=<2 -2 0>, rotation=<rot 90 0 0>}]
 dyn_wire_up [{shapeType=Capsule, length=2, radius=1, shapeName=dyn_wire_up_capsule, center=<2 -2 0>, rotation=<rot 90 0 0>}]
}
constraints [
 {bone1Name=dyn_fur_00, bone2Name=base_fur_00, swing1LimitDegrees=10, swing2LimitDegrees=6, twistLimitDegrees=0}
 {bone1Name=dyn_fur_01, bone2Name=base_fur_01, swing1LimitDegrees=11, swing2LimitDegrees=4, twistLimitDegrees=0}
 {bone1Name=dyn_wire_dwn, bone2Name=base_wire_dwn, swing1LimitDegrees=80, swing2LimitDegrees=5, twistLimitDegrees=1}
 {bone1Name=dyn_wire_up, bone2Name=base_wire_up, swing1LimitDegrees=80, swing2LimitDegrees=5, twistLimitDegrees=1}
]
